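# Sound cue for ambient musical one-shot stingers at the Season 39 Secret Facility POI, whose waves are named Area51.
# An equally weighted Random node picks one of ten stingers, passed through a Modulator that keeps pitch fixed.
# It uses the musical one-shot no-reverb, low-EQ attenuation and the in-game SFX sound class.
# A concurrency override allows one instance at a time and blocks new triggers while it plays.
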SoundCue /Hera/Audio/MusicalOneshots/S39_POI_SecretFacility_AmbientOneShot/S39_MusicalOneShots_SecretFacility_Cue
root=Modulator_0
nodes [
 {id=Modulator_0, children [Random_2], properties {pitchMax=1, pitchMin=1, volumeMax=1}}
 {
  id=Random_2
  children [
   WavePlayer_20
   WavePlayer_21
   WavePlayer_22
   WavePlayer_23
   WavePlayer_24
   WavePlayer_25
   WavePlayer_26
   WavePlayer_27
   WavePlayer_28
   WavePlayer_29
  ]
  properties {weights [1,1,1,1,1,1,1,1,1,1]}
 }
 {
  id=WavePlayer_20
  properties {soundWaveAssetPtr=&/MX_S39_POI_Area51_AmbientOneShot_STNG_10}
 }
 {
  id=WavePlayer_21
  properties {soundWaveAssetPtr=&/MX_S39_POI_Area51_AmbientOneShot_STNG_9}
 }
 {
  id=WavePlayer_22
  properties {soundWaveAssetPtr=&/MX_S39_POI_Area51_AmbientOneShot_STNG_8}
 }
 {
  id=WavePlayer_23
  properties {soundWaveAssetPtr=&/MX_S39_POI_Area51_AmbientOneShot_STNG_7}
 }
 {
  id=WavePlayer_24
  properties {soundWaveAssetPtr=&/MX_S39_POI_Area51_AmbientOneShot_STNG_6}
 }
 {
  id=WavePlayer_25
  properties {soundWaveAssetPtr=&/MX_S39_POI_Area51_AmbientOneShot_STNG_5}
 }
 {
  id=WavePlayer_26
  properties {soundWaveAssetPtr=&/MX_S39_POI_Area51_AmbientOneShot_STNG_4}
 }
 {
  id=WavePlayer_27
  properties {soundWaveAssetPtr=&/MX_S39_POI_Area51_AmbientOneShot_STNG_3}
 }
 {
  id=WavePlayer_28
  properties {soundWaveAssetPtr=&/MX_S39_POI_Area51_AmbientOneShot_STNG_2}
 }
 {
  id=WavePlayer_29
  properties {soundWaveAssetPtr=&/MX_S39_POI_Area51_AmbientOneShot_STNG_1}
 }
]
properties {
 attenuationSettings=/AmbientAudioBase/Attenuations/Fort_Ambient_Athena_Musical_Oneshot_NoVerb_LowEQ_Att
 bOverrideConcurrency=true
 maxDistance=2097152
 soundClassObject=&Fort_SFX_InGame
 volumeMultiplier=1
 concurrencyOverrides {maxCount=1, resolutionRule=PreventNew, voiceStealReleaseTime=2}
}
