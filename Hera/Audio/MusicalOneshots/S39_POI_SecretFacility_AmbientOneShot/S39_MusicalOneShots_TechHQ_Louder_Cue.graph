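# Sound Cue for musical ambient one-shot stingers at the Season 39 TechHQ secret facility POI, in a boosted-volume "Louder" version.
# An equal-weight Random node picks from ten stinger waves and feeds a Modulator that leaves pitch unaltered.
# A no-reverb, low-EQ musical one-shot attenuation shapes it, and it routes to the in-game SFX sound class.
# A concurrency override allows one instance at a time, blocking new plays rather than interrupting.
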SoundCue /Hera/Audio/MusicalOneshots/S39_POI_SecretFacility_AmbientOneShot/S39_MusicalOneShots_TechHQ_Louder_Cue
root=Modulator_0
nodes [
 {id=Modulator_0, children [Random_1], properties {pitchMax=1, pitchMin=1, volumeMax=1}}
 {
  id=Random_1
  children [WavePlayer_0, WavePlayer_1, WavePlayer_2, WavePlayer_3, WavePlayer_4, WavePlayer_5, WavePlayer_6, WavePlayer_7, WavePlayer_10, WavePlayer_11]
  properties {weights [1,1,1,1,1,1,1,1,1,1]}
 }
 {
  id=WavePlayer_0
  properties {soundWaveAssetPtr=&/MX_S39_POI_TechHQ_OneShots_STNG_01}
 }
 {
  id=WavePlayer_1
  properties {soundWaveAssetPtr=&/MX_S39_POI_TechHQ_OneShots_STNG_02}
 }
 {
  id=WavePlayer_10
  properties {soundWaveAssetPtr=&/MX_S39_POI_TechHQ_OneShots_STNG_09}
 }
 {
  id=WavePlayer_11
  properties {soundWaveAssetPtr=&/MX_S39_POI_TechHQ_OneShots_STNG_10}
 }
 {
  id=WavePlayer_2
  properties {soundWaveAssetPtr=&/MX_S39_POI_TechHQ_OneShots_STNG_03}
 }
 {
  id=WavePlayer_3
  properties {soundWaveAssetPtr=&/MX_S39_POI_TechHQ_OneShots_STNG_04}
 }
 {
  id=WavePlayer_4
  properties {soundWaveAssetPtr=&/MX_S39_POI_TechHQ_OneShots_STNG_05}
 }
 {
  id=WavePlayer_5
  properties {soundWaveAssetPtr=&/MX_S39_POI_TechHQ_OneShots_STNG_06}
 }
 {
  id=WavePlayer_6
  properties {soundWaveAssetPtr=&/MX_S39_POI_TechHQ_OneShots_STNG_07}
 }
 {
  id=WavePlayer_7
  properties {soundWaveAssetPtr=&/MX_S39_POI_TechHQ_OneShots_STNG_08}
 }
]
properties {
 attenuationSettings=/AmbientAudioBase/Attenuations/Fort_Ambient_Athena_Musical_Oneshot_NoVerb_LowEQ_Att
 bOverrideConcurrency=true
 maxDistance=2097152
 soundClassObject=&Fort_SFX_InGame
 volumeMultiplier=1.5
 concurrencyOverrides {maxCount=1, resolutionRule=PreventNew, voiceStealReleaseTime=2}
}
